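# Translucent unlit Niagara sprite material for the EctoCat character's glob effects, drawing soft, wobbling blobs in a two-color gradient.
# A panning Gaussian-noise normal map, offset by particle random, distorts UVs feeding a radial distance mask with adjustable size.
# The mask blends the colors, tinted by particle color through an emissive-adjustment function.
# The powered mask gets depth fade, camera fade and particle alpha, which a shading-path switch skips on forward and mobile.
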Material /Game/Characters/Player/Male/Large/Bodies/M_LRG_EctoCat/FX/M_Character_EctoCat_Globs
graph {
 properties {bUsedWithNiagaraSprites=true, blendMode=BLEND_Translucent, pixelDepthOffsetMode=PDOM_Legacy, shadingModel=MSM_Unlit, usedShadingModels=Unlit}
 nodes [
  Constant() {r=1}
  n1 = TextureCoordinate()
  n2 = ParticleRandom()
  n3 = Multiply(A=TextureCoordinate(), B=ScalarParameter() {defaultValue=0.1, parameterName=Noise Scale})
  n4 = Constant2Vector() {g=0.08, r=0.5}
  n5 = ScalarParameter() {defaultValue=0.5, parameterName=Param_1}
  n6 = TextureSample(UVs=Panner(Coordinate=Add(A=@n2, B=@n3), Speed=Multiply(A=@n4, B=@n5)) {speedX=0.3, speedY=-0.5}) {samplerType=SAMPLERTYPE_Normal, texture=/Game/Characters/Player/Female/Medium/Bodies/F_MED_Bistro_Astronaut/FX/T_GaussianNoise_Normal}
  n7 = Multiply(A=ComponentMask(None=@n6.RGB) {g=true, r=true}, B=ScalarParameter() {defaultValue=0.4, parameterName=B}) {constB=0.1}
  n8 = Multiply(A=Distance(A=Add(A=@n1, B=@n7) {constB=0}, B=Constant() {r=0.5})) {constB=2}
  n9 = Subtract(A=OneMinus(None=Saturate(None=@n8)), B=ScalarParameter() {defaultValue=0.5, parameterName=Size})
  n10 = LinearInterpolate(
   A=VectorParameter() {parameterName=v, defaultValue=<rgba 0.43 0.0495 0.45 1>}.RGB
   B=VectorParameter() {parameterName=v2, defaultValue=<rgba 0.9 0.1701 0.7297 1>}.RGB
   Alpha=@n9
  )
  n11 = Ceil(None=@n9)
  MaterialFunctionCall(Emissive=Multiply(A=LinearInterpolate(B=@n10, Alpha=@n11), B=ParticleColor().RGB)) {materialFunction=&MF_FX_EmissiveAdjustments}
  n12 = ScalarParameter() {defaultValue=15, parameterName=FadeDistance}
  n13 = Power(Base=@n9, Exp=ScalarParameter() {defaultValue=0.1, parameterName=Exp})
  n14 = ScalarParameter() {defaultValue=40, parameterName="Fade Length (S)"}
  n15 = ScalarParameter() {defaultValue=120, parameterName="Fade Offset (S)"}
  n16 = Multiply(
   A=Multiply(
    A=DepthFade(Opacity=@n13, FadeDistance=@n12)
    B=MaterialFunctionCall(Fade Length=@n14, Fade Offset=@n15) {materialFunction=&CameraDepthFade}.Result
   )
   B=ParticleColor().A
  )
  ShadingPathSwitch(Default=@n16, Deferred=@n16, Forward=@n13, Mobile=@n13)
 ]
}
