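# Idle goop effect for the EctoCat player character, emitting short-lived mesh and sprite particles from the character's skeletal mesh.
# Owner speed lowers spawn rate and lifetime; a camera-distance "ADS" flag and hidden-hand state choose which bone/socket ranges emit.
# Particles pop and shrink via size curves, briefly follow source movement, then drift upward under drag and rising Z force.
# Color scaling zeroes alpha in vehicles and dims RGB at night.
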NiagaraSystem /Game/Characters/Player/Male/Large/Bodies/M_LRG_EctoCat/FX/NS_Character_EctoCat_Idle
dependencies {
 usedDataInterfaces [&NiagaraDataInterfaceCamera, &NiagaraDataInterfaceCurve, &NiagaraDataInterfaceSkeletalMesh]
 usedDynamicInputs []
 usedModules [
  &SystemState
  &self:SystemSpawnScript.SystemScriptSource.SystemScriptGraph.NiagaraNodeAssignment_0.SetVariables_#1
 ]
}
emitters {Goop_Base {
  emitterUpdateScript [
   &EmitterState(Inactive Response=::NewEnumerator1, Life Cycle Mode=::NewEnumerator1, Loop Duration=2, UseLoopDelay=false)
   &SpawnRate(
    Spawn Probability=0.6
    SpawnRate=&Lerp_Float(
     A=5
     B=1.5
     Alpha=&ClampFloat(Float=&Divide_Float(A=&VectorLength(Vector=$Engine.Owner.Velocity), B=500))
    )
    Use Spawn Probability=true
   )
   /SetVariables(Emitter.ADS:NiagaraBool=&SetBoolByFloatComparison(
     A=&DistanceBetweenPositions(
      A=&ReturnCameraPosition(
       Coordinate Space:ENiagaraCoordinateSpace=::Simulation
       Camera Query:NiagaraDataInterfaceCamera=DataInterface() {propertyValues {bRequireCurrentFrameData=false}}
      )
      B=$Engine.Emitter.SimulationPosition
     )
     B=&FN_FrontEndFloatModifier(Gameplay=250, Front End=0)
     Comparison Type=::NewEnumerator6
    )) {isSetParameters=true}
  ]
  particleSpawnScript [
   &PartitionParticles()
   &InitializeParticle(
    Color=<rgba 0.8 0.8 0.8 3>
    Lifetime=&Lerp_Float(
     A=&RandomRangeFloat(Minimum=1.4, Maximum=2.2)
     B=0.8
     Alpha=&Divide_Float(A=&VectorLength(Vector=$Engine.Owner.Velocity), B=5000)
    )
    Mesh Scale Mode=::NewEnumerator1
    Mesh Uniform Scale Max=0.4
    Mesh Uniform Scale Min=0.25
    Sprite Size Mode=::NewEnumerator1
    UsePositionOffset=false
    Write Material Random=true
   )
   &InitialMeshOrientation(Use Rotation Vector=false)
   &SkeletalMeshLocation(
    Bone / Socket Index=&MakeCustomIntFromBool(
     Bool=$Emitter.ADS
     True INT=&RandomRangeInt(Minimum=7, Maximum=11)
     False INT=&MakeCustomIntFromBool(
      Bool=$User.bisHandHidden
      True INT=&MakeCustomIntFromBool(
       Bool=&SetBoolByIntComparison(A=$Particles.PartitionParticles.Partition, B=0, Comparison Type=::NewEnumerator0)
       True INT=&RandomRangeInt(Minimum=19, Maximum=23)
       False INT=&RandomRangeInt(Minimum=7, Maximum=11)
      )
      False INT=&MakeCustomIntFromBool(
       Bool=&SetBoolByIntComparison(A=$Particles.PartitionParticles.Partition, B=0, Comparison Type=::NewEnumerator0)
       True INT=&RandomRangeInt(Minimum=12, Maximum=23)
       False INT=&RandomRangeInt(Minimum=0, Maximum=11)
      )
     )
    )
    MaskSpawnGroup=false
    Mesh Sampling Type=::NewEnumerator5
    Skeletal Mesh=$System.SkeletalMesh
    Socket Sampling Mode=::NewEnumerator5
   )
   &AddVelocity(
    Coordinate Space=::World
    Velocity=&Lerp_Vector(
     A=<0 0 10>
     B=<0 0 30>
     Alpha=&Divide_Float(A=&VectorLength(Vector=$Engine.Owner.Velocity), B=5000)
    )
   ) {enabled=false, module=AddVelocity001}
   &AddVelocity(Module.Use Falloff Distance=false, Override Seed=false, Velocity Mode=::NewEnumerator1, Velocity Speed=100)
   &InheritVelocity(Speed Limit=true, UseSpeedThreshold=false)
   /SetVariables(Particles.MaterialRandom:NiagaraFloat=&RandomRangeFloat(Minimum=0, Maximum=1)) {isSetParameters=true}
  ]
  particleUpdateScript [
   &ParticleState()
   &ScaleSpriteSize(
    Scale Sprite Size Mode=::NewEnumerator0
    Uniform Scale Factor=&FloatFromCurve(FloatCurve=DataInterface() {propertyValues {Curve {keys [
         {arriveTangent=9.1187, interpMode=RCIM_Cubic, time=0, value=0}
         {arriveTangent=0.3759, interpMode=RCIM_Cubic, leaveTangent=0.3759, time=0.2487, value=1.1246}
         {arriveTangent=-1.497, interpMode=RCIM_Cubic, leaveTangent=-1.497, time=0.7835, value=0.2945}
         <at 1 0 RCIM_Cubic>
        ]}}})
   )
   &ScaleMeshSize(Scale Factor=&VectorFromFloat(Value=&FloatFromCurve(
      FloatCurve=DataInterface() {propertyValues {Curve {keys [
          {arriveTangent=9.1187, interpMode=RCIM_Cubic, time=0, value=0}
          {arriveTangent=0.3759, interpMode=RCIM_Cubic, leaveTangent=0.3759, time=0.2487, value=1.1246}
          {arriveTangent=-1.497, interpMode=RCIM_Cubic, leaveTangent=-1.497, time=0.7835, value=0.2945}
          <at 1 0 RCIM_Cubic>
         ]}}}
      Scale Curve=0.18
     ))) {module=ScaleMeshSize001}
   &AccelerationForce(Acceleration=&Multiply_VectorByFloat(
     Vector=&RandomRangeVector(Minimum=<-3 -3 5>, Maximum=<3 3 10>, Evaluation Type=::NewEnumerator1)
     Float=1
    ))
   &MeshRotationForce(
    Coordinate Space:ENiagaraMeshTransforms=::NewEnumerator3
    Rotation=&RandomRangeVector(Minimum=<-4 -4 -4>, Maximum=<4 4 4>)
   )
   &Drag(Drag=5, Use Linear Drag=true, Use Rotational Drag=true)
   &GravityForce(Gravity=&MakeVector(
     X=0
     Y=0
     Z=&FloatFromCurve(FloatCurve=DataInterface() {propertyValues {Curve {keys [<at 0.3 0>,<at 1 1>]}}}, Scale Curve=400)
    ))
   &InheritSourceMovement(
    Applied Force Scale=&VectorFromFloat(Value=&FloatFromCurve(
      FloatCurve=DataInterface() {propertyValues {Curve {keys [
          {interpMode=RCIM_Cubic, tangentMode=RCTM_User, time=0, value=1}
          {arriveTangent=-7.0714, interpMode=RCIM_Cubic, leaveTangent=-7.0714, tangentMode=RCTM_User, time=0.3, value=0}
         ]}}}
      Scale Curve=6
     ))
    Applied Position Scale=&VectorFromFloat(Value=&FloatFromCurve(
      FloatCurve=DataInterface() {propertyValues {Curve {keys [
          {interpMode=RCIM_Cubic, tangentMode=RCTM_User, time=0.5, value=1}
          {arriveTangent=-4.8956, interpMode=RCIM_Cubic, leaveTangent=-4.8956, tangentMode=RCTM_User, time=1, value=0}
         ]}}}
      Scale Curve=0.2
     ))
    ApplyForce=true
    ApplyForceFalloff=false
    ApplyPosition=true
    ApplyPositionFalloff=false
    Module.LimitForceSpeed=false
   )
   &DynamicMaterialParameters(
    Index 0 Param 1=5
    Index 0 Param 2=12
    Index 0 Param 3=&FloatFromCurve(FloatCurve=DataInterface() {propertyValues {Curve {keys [<at 0 5 RCIM_Cubic>, {arriveTangent=-4.1151, interpMode=RCIM_Cubic, leaveTangent=-4.1151, tangentMode=RCTM_User, time=1, value=1}]}}})
    Module.Param0WriteEnabled=true
    Module.Param1WriteEnabled=true
    Module.Param2WriteEnabled=true
    Module.Param3WriteEnabled=false
   )
   &SolveForcesAndVelocity(Clamp Velocity=false, Limit Acceleration=false)
   /SetVariables(Particles.VisibilityTag=&MakeCustomIntFromBool(
     Bool=&SetBoolByFloatComparison(
      A=&DistanceBetweenPositions(
       A=&ReturnCameraPosition(
        Coordinate Space:ENiagaraCoordinateSpace=::Simulation
        Camera Query:NiagaraDataInterfaceCamera=DataInterface() {propertyValues {bRequireCurrentFrameData=false}}
       )
       B=$Particles.Position
      )
      B=&FN_FrontEndFloatModifier(Gameplay=150, Front End=0)
      Comparison Type=::NewEnumerator6
     )
     True INT=1
     False INT=0
    )) {enabled=false, isSetParameters=true}
   &ScaleColor(Scale Alpha=$Particles.Timeline.InverseNormalizedTime, ScaleA=true, ScaleRGB=true) {enabled=false}
   &ScaleColor(
    Scale Alpha=&MakeCustomFloatFromBool(Bool=$User.bisInVehicle, True Float=0, False Float=1)
    ScaleA=true
    ScaleRGB=true
   ) {module=ScaleColor001}
   &ScaleColor(
    Scale RGB=&VectorFromFloat(Value=&FNTimeOfDayScalar(Night Scale=0.6, Morning Scale=1, Evening Scale=1))
    ScaleA=true
    ScaleRGB=true
   ) {module=ScaleColor002}
  ]
  rendererClasses [&NiagaraMeshRendererProperties, &NiagaraSpriteRendererProperties]
  renderers [
   {rendererClass=&NiagaraMeshRendererProperties, rendererIndex=0}
   {rendererClass=&NiagaraSpriteRendererProperties, rendererIndex=1}
  ]
 }}
summary {
 systemName=NS_Character_EctoCat_Idle
 emitters [{emitterName=Goop_Base, rendererClasses [&NiagaraMeshRendererProperties, &NiagaraSpriteRendererProperties]}]
 userVariables [
  User.bIsFrontEnd {type=NiagaraBool}
  User.bIsFrontEndPreview {type=NiagaraBool}
  User.bisHandHidden {type=NiagaraBool}
  User.bisInVehicle {type=NiagaraBool}
 ]
}
