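# Post-process style AnimBlueprint for the EctoCat large male player body, adding secondary motion to the pose copied from the parent mesh.
# The pose runs through a deform Control Rig, AnimDynamics pant bones, then additive belt corrections weighted by thigh-to-pelvis deltas.
# The event graph retunes dynamics and upper-arm deformation for skydiving.
# It also feeds leg motion into material WPO regions and enables cloth wind while skydive-floating.
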AnimBlueprint /Game/Characters/Player/Male/Large/Bodies/M_LRG_EctoCat/Meshes/M_LRG_EctoCat_AnimBP
target_skeleton=&/M_LRG_EctoCat_Skeleton
graphs [
 AnimGraph {nodes [
   n1 = Control Rig!(
    Source=Use cached pose 'SavedPose27310'()
    Alpha
    DeformRig_Value_upperarm_up_Translation_Z:double=Get Deform Rig Value Upperarm Up Translation Z() -> (Deform Rig Value Upperarm Up Translation Z:double)
   ) -> (Pose) {
    alphaBoolBlend {blendOption=Linear}
    controlRigAssetReference {blueprintRigClass=&/M_LRG_EctoCat_CtrlRig.M_LRG_EctoCat_CtrlRig_C}
   }
   n2 = Get Component Linear Vel Scale() -> (Component Linear Vel Scale:Vector)
   n3 = Get Component Linear Acc Scale() -> (Component Linear Acc Scale:Vector)
   n4 = Anim Dynamics - Bone: dyn_pant_fr_1_l!(
    ComponentPose=Local To Component(LocalPose=@n1)
    Alpha=1
    ComponentLinearAccScale:Vector=@n3
    ComponentLinearVelScale:Vector=@n2
    ExternalForce="0,0,0"
   ) -> (Pose) {
    angularDampingOverride=0.999
    angularSpringConstant=2000
    bAngularSpring=true
    bOverrideAngularDamping=true
    bUseGravityOverride=true
    simulationSpace=BoneRelative
    alphaBoolBlend {blendOption=Linear}
    boundBone {boneName=dyn_pant_fr_1_l}
    componentAppliedLinearAccClamp=<2000 2000 2000>
    componentLinearAccScale=<0.3 0.3 0.3>
    componentLinearVelScale=<3 3 3>
    gravityOverride {z=-980}
    physicsBodyDefinitions [{
      boundBone {boneName=dyn_pant_fr_1_l}
      boxExtents=<8 8 8>
      constraintSetup {angularLimitsMax=<2 5 20>, angularLimitsMin=<-2 -5 -20>, angularTarget=<0 0 0>}
      localJointOffset=<1 0 0>
     }]
   }
   n5 = Anim Dynamics - Bone: dyn_pant_fr_2_l!(ComponentPose=@n4, Alpha=1, ComponentLinearAccScale:Vector=@n3, ComponentLinearVelScale:Vector=@n2, ExternalForce="0,0,0") -> (Pose) {
    angularDampingOverride=0.999
    angularSpringConstant=2000
    bAngularSpring=true
    bOverrideAngularDamping=true
    bUseGravityOverride=true
    simulationSpace=BoneRelative
    alphaBoolBlend {blendOption=Linear}
    boundBone {boneName=dyn_pant_fr_2_l}
    componentAppliedLinearAccClamp=<2000 2000 2000>
    componentLinearAccScale=<0.3 0.3 0.3>
    componentLinearVelScale=<3 3 3>
    gravityOverride {z=-980}
    physicsBodyDefinitions [{
      boundBone {boneName=dyn_pant_fr_2_l}
      boxExtents=<8 8 8>
      constraintSetup {angularLimitsMax=<2 5 20>, angularLimitsMin=<-2 -5 -20>, angularTarget=<0 0 0>}
      localJointOffset=<1 0 0>
     }]
   }
   n6 = Anim Dynamics - Bone: dyn_pant_bck_1_l!(ComponentPose=@n5, Alpha=1, ComponentLinearAccScale:Vector=@n3, ComponentLinearVelScale:Vector=@n2, ExternalForce="0,0,0") -> (Pose) {
    angularDampingOverride=0.999
    angularSpringConstant=2000
    bAngularSpring=true
    bOverrideAngularDamping=true
    bUseGravityOverride=true
    simulationSpace=BoneRelative
    alphaBoolBlend {blendOption=Linear}
    boundBone {boneName=dyn_pant_bck_1_l}
    componentAppliedLinearAccClamp=<2000 2000 2000>
    componentLinearAccScale=<0.3 0.3 0.3>
    componentLinearVelScale=<3 3 3>
    gravityOverride {z=-980}
    physicsBodyDefinitions [{
      boundBone {boneName=dyn_pant_bck_1_l}
      boxExtents=<8 8 8>
      constraintSetup {angularLimitsMax=<2 5 20>, angularLimitsMin=<-2 -5 -20>, angularTarget=<0 0 0>}
      localJointOffset=<1 0 0>
     }]
   }
   n7 = Anim Dynamics - Bone: dyn_pant_bck_2_l!(ComponentPose=@n6, Alpha=1, ComponentLinearAccScale:Vector=@n3, ComponentLinearVelScale:Vector=@n2, ExternalForce="0,0,0") -> (Pose) {
    angularDampingOverride=0.999
    angularSpringConstant=2000
    bAngularSpring=true
    bOverrideAngularDamping=true
    bUseGravityOverride=true
    simulationSpace=BoneRelative
    alphaBoolBlend {blendOption=Linear}
    boundBone {boneName=dyn_pant_bck_2_l}
    componentAppliedLinearAccClamp=<2000 2000 2000>
    componentLinearAccScale=<0.3 0.3 0.3>
    componentLinearVelScale=<3 3 3>
    gravityOverride {z=-980}
    physicsBodyDefinitions [{
      boundBone {boneName=dyn_pant_bck_2_l}
      boxExtents=<8 8 8>
      constraintSetup {angularLimitsMax=<2 5 20>, angularLimitsMin=<-2 -5 -20>, angularTarget=<0 0 0>}
      localJointOffset=<1 0 0>
     }]
   }
   n8 = Anim Dynamics - Bone: dyn_pant_sd_l!(ComponentPose=@n7, Alpha=1, ComponentLinearAccScale:Vector=@n3, ComponentLinearVelScale:Vector=@n2, ExternalForce="0,0,0") -> (Pose) {
    angularDampingOverride=0.999
    angularSpringConstant=2000
    bAngularSpring=true
    bOverrideAngularDamping=true
    bUseGravityOverride=true
    simulationSpace=BoneRelative
    alphaBoolBlend {blendOption=Linear}
    boundBone {boneName=dyn_pant_sd_l}
    componentAppliedLinearAccClamp=<2000 2000 2000>
    componentLinearAccScale=<0.3 0.3 0.3>
    componentLinearVelScale=<3 3 3>
    gravityOverride {z=-980}
    physicsBodyDefinitions [{
      boundBone {boneName=dyn_pant_sd_l}
      boxExtents=<8 8 8>
      constraintSetup {angularLimitsMax=<2 5 20>, angularLimitsMin=<-2 -5 -20>, angularTarget=<0 0 0>}
      localJointOffset=<1 0 0>
     }]
   }
   n9 = Anim Dynamics - Bone: dyn_pant_fr_1_r!(ComponentPose=@n8, Alpha=1, ComponentLinearAccScale:Vector=@n3, ComponentLinearVelScale:Vector=@n2, ExternalForce="0,0,0") -> (Pose) {
    angularDampingOverride=0.999
    angularSpringConstant=2000
    bAngularSpring=true
    bOverrideAngularDamping=true
    bUseGravityOverride=true
    simulationSpace=BoneRelative
    alphaBoolBlend {blendOption=Linear}
    boundBone {boneName=dyn_pant_fr_1_r}
    componentAppliedLinearAccClamp=<2000 2000 2000>
    componentLinearAccScale=<0.3 0.3 0.3>
    componentLinearVelScale=<3 3 3>
    gravityOverride {z=-980}
    physicsBodyDefinitions [{
      boundBone {boneName=dyn_pant_fr_1_r}
      boxExtents=<8 8 8>
      constraintSetup {angularLimitsMax=<2 5 20>, angularLimitsMin=<-2 -5 -20>, angularTarget=<0 0 0>}
      localJointOffset=<1 0 0>
     }]
   }
   n10 = Anim Dynamics - Bone: dyn_pant_fr_2_r!(ComponentPose=@n9, Alpha=1, ComponentLinearAccScale:Vector=@n3, ComponentLinearVelScale:Vector=@n2, ExternalForce="0,0,0") -> (Pose) {
    angularDampingOverride=0.999
    angularSpringConstant=2000
    bAngularSpring=true
    bOverrideAngularDamping=true
    bUseGravityOverride=true
    simulationSpace=BoneRelative
    alphaBoolBlend {blendOption=Linear}
    boundBone {boneName=dyn_pant_fr_2_r}
    componentAppliedLinearAccClamp=<2000 2000 2000>
    componentLinearAccScale=<0.3 0.3 0.3>
    componentLinearVelScale=<3 3 3>
    gravityOverride {z=-980}
    physicsBodyDefinitions [{
      boundBone {boneName=dyn_pant_fr_2_r}
      boxExtents=<8 8 8>
      constraintSetup {angularLimitsMax=<2 5 20>, angularLimitsMin=<-2 -5 -20>, angularTarget=<0 0 0>}
      localJointOffset=<1 0 0>
     }]
   }
   n11 = Anim Dynamics - Bone: dyn_pant_bck_2_r!(ComponentPose=@n10, Alpha=1, ComponentLinearAccScale:Vector=@n3, ComponentLinearVelScale:Vector=@n2, ExternalForce="0,0,0") -> (Pose) {
    angularDampingOverride=0.999
    angularSpringConstant=2000
    bAngularSpring=true
    bOverrideAngularDamping=true
    bUseGravityOverride=true
    simulationSpace=BoneRelative
    alphaBoolBlend {blendOption=Linear}
    boundBone {boneName=dyn_pant_bck_2_r}
    componentAppliedLinearAccClamp=<2000 2000 2000>
    componentLinearAccScale=<0.3 0.3 0.3>
    componentLinearVelScale=<3 3 3>
    gravityOverride {z=-980}
    physicsBodyDefinitions [{
      boundBone {boneName=dyn_pant_bck_2_r}
      boxExtents=<8 8 8>
      constraintSetup {angularLimitsMax=<2 5 20>, angularLimitsMin=<-2 -5 -20>, angularTarget=<0 0 0>}
      localJointOffset=<1 0 0>
     }]
   }
   n12 = Anim Dynamics - Bone: dyn_pant_sd_r!(ComponentPose=@n11, Alpha=1, ComponentLinearAccScale:Vector=@n3, ComponentLinearVelScale:Vector=@n2, ExternalForce="0,0,0") -> (Pose) {
    angularDampingOverride=0.999
    angularSpringConstant=2000
    bAngularSpring=true
    bOverrideAngularDamping=true
    bUseGravityOverride=true
    simulationSpace=BoneRelative
    alphaBoolBlend {blendOption=Linear}
    boundBone {boneName=dyn_pant_sd_r}
    componentAppliedLinearAccClamp=<2000 2000 2000>
    componentLinearAccScale=<0.3 0.3 0.3>
    componentLinearVelScale=<3 3 3>
    gravityOverride {z=-980}
    physicsBodyDefinitions [{
      boundBone {boneName=dyn_pant_sd_r}
      boxExtents=<8 8 8>
      constraintSetup {angularLimitsMax=<2 5 20>, angularLimitsMin=<-2 -5 -20>, angularTarget=<0 0 0>}
      localJointOffset=<1 0 0>
     }]
   }
   n13 = Get Left_leg_belt() -> (Left_leg_belt:double)
   n14 = "Transform (Modify) Bone - Bone: dyn_belt_l"(ComponentPose=@n12, Alpha=@n13, Translation="-9,-1,0", Rotation="-13.287,0.014,-0.000", Scale="1,1,1") {
    rotationMode=BMM_Additive
    rotationSpace=BCS_BoneSpace
    translationMode=BMM_Additive
    translationSpace=BCS_BoneSpace
    alphaBoolBlend {blendOption=Linear}
    boneToModify {boneName=dyn_belt_l}
    rotation=<rot -13.287 0.014 0>
    translation=<-9 -1>
   }
   n15 = "Transform (Modify) Bone - Bone: dyn_belt_fr"(ComponentPose=@n14, Alpha=@n13, Translation="-2.5,0,0", Rotation="18,0,0", Scale="1,1,1") {
    rotationMode=BMM_Additive
    rotationSpace=BCS_BoneSpace
    translationMode=BMM_Additive
    translationSpace=BCS_BoneSpace
    alphaBoolBlend {blendOption=Linear}
    boneToModify {boneName=dyn_belt_fr}
    rotation {pitch=18}
    translation {x=-2.5}
   }
   n16 = "Transform (Modify) Bone - Bone: dyn_belt_r"(
    ComponentPose=@n15
    Alpha=Get Right_leg_belt_crouch() -> (Right_leg_belt_crouch:double)
    Translation="-10,0,0"
    Rotation="0,0,0"
    Scale="1,1,1"
   ) {
    rotationMode=BMM_Additive
    rotationSpace=BCS_BoneSpace
    translationMode=BMM_Additive
    translationSpace=BCS_BoneSpace
    alphaBoolBlend {blendOption=Linear}
    boneToModify {boneName=dyn_belt_r}
    translation {x=-10}
   }
   n17 = Get Right_leg_belt() -> (Right_leg_belt:double)
   n18 = "Transform (Modify) Bone - Bone: dyn_belt_r"(ComponentPose=@n16, Alpha=@n17, Translation="-1,-3,0", Rotation="0,0,0", Scale="1,1,1") {
    rotationMode=BMM_Additive
    rotationSpace=BCS_BoneSpace
    translationMode=BMM_Additive
    translationSpace=BCS_BoneSpace
    alphaBoolBlend {blendOption=Linear}
    boneToModify {boneName=dyn_belt_r}
    translation=<-1 -3>
   }
   n19 = "Transform (Modify) Bone - Bone: dyn_belt_fr"(ComponentPose=@n18, Alpha=@n17, Translation="0,0,0", Rotation="-20,0,0", Scale="1,1,1") {
    rotationMode=BMM_Additive
    rotationSpace=BCS_BoneSpace
    translationMode=BMM_Additive
    translationSpace=BCS_BoneSpace
    alphaBoolBlend {blendOption=Linear}
    boneToModify {boneName=dyn_belt_fr}
    rotation {pitch=-20}
   }
   Output Pose(Result=Component To Local(ComponentPose=@n19)) {layerGroup=None}
   Save cached pose 'SavedPose27310'(Pose=Copy Pose From Mesh() {bUseAttachedParent=true}) {cacheName=SavedPose27310}
  ]}
 EventGraph {nodes [
   n1 = Event BlueprintUpdateAnimation()
   > n2 = Sequence!(execute) -> (then_0, then_1, then_2:Exec, then_3:Exec)
   n3 = Get Skel Mesh()
   > Set Left_leg_belt(Left_leg_belt:double=GetAlphaFromReferenceTransformDelta(SkeletalMeshComponent=@n3, BoneName=thigh_l, BaseName=pelvis, InputRange=(X=0,Y=-92), TwistAxis="1,0,0")) -> (Output_Get:double)
   > Set Right_leg_belt(Right_leg_belt:double=GetAlphaFromReferenceTransformDelta(SkeletalMeshComponent=@n3, BoneName=thigh_r, BaseName=pelvis, DeltaComponentType=EulerY, InputRange=(X=7,Y=42), TwistAxis="1,0,0")) -> (Output_Get:double)
   > Set Right_leg_belt_crouch(Right_leg_belt_crouch:double=GetAlphaFromReferenceTransformDelta(SkeletalMeshComponent=@n3, BoneName=thigh_r, BaseName=pelvis, InputRange=(X=-18,Y=-98), TwistAxis="1,0,0")) -> (Output_Get:double)
   n4 = Set Delta Time X(execute=@n2.then_1, Delta Time X=@n1.DeltaTimeX)
   > Is Valid(InputObject=Get Custom Base Mesh() -> (Custom Base Mesh:Skeletal Mesh Component Object Reference))
   n5 = Get Mat_Instance Dynamic() -> (Mat_Instance Dynamic:Material Instance Dynamic Object Reference)
   n6 = Get Overlay_Mat_Inst() -> (Overlay_Mat_Inst:Material Instance Dynamic Object Reference)
   n7 = Get Skel Mesh()
   > Set WPO Globals Function(
    Material=@n5
    Overlay Material=@n6
    Skeletal Mesh=@n7
    "Local Noise Transform  Location"="Get Local Noise Transform  Location"() -> ("Local Noise Transform  Location":Vector)
    Local Noise Offset=Get Local Noise Offset()
   )
   > Set WPO Regional Function!(
    execute
    Material:Material Instance Dynamic Object Reference=@n5
    Overlay Material:Material Instance Dynamic Object Reference=@n6
    Skeletal Mesh:Skeletal Mesh Component Object Reference=@n7
    Velocity Transform:Name=foot_l
    Velocity Interp Speed:double=Get Velocity Interp Speed()
    Collision Transform:Name=calf_l
    Collision Transform Axis:EAxis Enum
    Invert Collision Transform Axis:bool=true
    Delta Time X:double=@n4
    Previous Velocity Transform Location:"Vector (by ref)"=Get R1_Previous Velocity Transform Location() -> (R1_Previous Velocity Transform Location:Vector)
    Previous Velocity Transform Velocity:"Vector (by ref)"=Get R1_Previous Velocity Transform Velocity() -> (R1_Previous Velocity Transform Velocity:Vector)
    Region:int=1
   ) -> (then)
   > Set WPO Regional Function(
    Material=@n5
    Overlay Material=@n6
    Skeletal Mesh=@n7
    Velocity Transform=foot_r
    Velocity Interp Speed=Get Velocity Interp Speed_0() -> (Velocity Interp Speed_0:double)
    Collision Transform=calf_r
    Invert Collision Transform Axis=false
    Delta Time X=@n4
    Previous Velocity Transform Location=Get R2_Previous Velocity Transform Location() -> (R2_Previous Velocity Transform Location:Vector)
    Previous Velocity Transform Velocity=Get R2_Previous Velocity Transform Velocity() -> (R2_Previous Velocity Transform Velocity:Vector)
    Region=2
   )
   > n8 = Branch(Condition=Get bIsSkydiveFloating())
   > n9 = Set WPO Pho Wind Intensity(WPO Pho Wind Intensity=1)
   SetScalarParameterValue(execute=[@n9.then, @n10.then], self=Get Mat_Instance Dynamic(), ParameterName=ClothSim_WindStrength, Value=Get WPO Pho Wind Intensity())
   n10 = Set WPO Pho Wind Intensity(execute=@n8.else, WPO Pho Wind Intensity=0)
   n11 = Branch(execute=@n2.then_2, Condition=Get bIsSkyDiving())
   > Set Component Linear Acc Scale(Component Linear Acc Scale:Vector="1,1,1") -> (Output_Get:Vector)
   > Set Component Linear Vel Scale(Component Linear Vel Scale:Vector="1,1,1") -> (Output_Get:Vector)
   Set Component Linear Acc Scale(execute=@n11.else, Component Linear Acc Scale="0.3,0.3,0.3")
   > Set Component Linear Vel Scale(Component Linear Vel Scale="3,3,3")
   n12 = Branch(execute=@n2.then_3, Condition=Get bIsSkyDiving())
   > Set Deform Rig Value Upperarm Up Translation Z(Deform Rig Value Upperarm Up Translation Z:double=5) -> (Output_Get:double)
   Set Deform Rig Value Upperarm Up Translation Z(execute=@n12.else, Deform Rig Value Upperarm Up Translation Z=8)
   n13 = Event OnCharacterCustomizationLoaded()
   > Set Custom Base Mesh(Custom Base Mesh:Skeletal Mesh Component Object Reference=@n13.CustomBaseMesh) -> (Output_Get:Skeletal Mesh Component Object Reference)
   > n14 = Set Skel Mesh(Skel Mesh=Get Array Item with EFort Custom Part Type(Array=@n13.CustomMeshParts))
   > n15 = Get Dynamic Material Instance(SkeletalMeshComponent=@n14)
   > Set Mat_Instance Dynamic(Mat_Instance Dynamic:Material Instance Dynamic Object Reference=@n15) -> (Output_Get:Material Instance Dynamic Object Reference)
   > n16 = GetOrCreateDynamicOverlayMaterial(MeshComponent=@n14, MaterialName=None)
   > Set Overlay_Mat_Inst(Overlay_Mat_Inst:Material Instance Dynamic Object Reference=@n16) -> (Output_Get:Material Instance Dynamic Object Reference)
  ]}
]
variables [
 Component Linear Acc Scale {type=Vector, default=<0.3 0.3 0.3>}
 Component Linear Vel Scale {type=Vector, default=<3 3 3>}
 Custom Base Mesh {type=SkeletalMeshComponent}
 Deform Rig Value Upperarm Up Translation Z {type=number}
 Delta Time X {type=number}
 Left_leg_belt {type=number}
 Local Noise Offset {type=Vector}
 "Local Noise Transform  Location" {type=Vector}
 Mat_Instance Dynamic {type=MaterialInstanceDynamic}
 Overlay_Mat_Inst {type=MaterialInstanceDynamic}
 R1_Previous Velocity Transform Location {type=Vector}
 R1_Previous Velocity Transform Velocity {type=Vector}
 R2_Previous Velocity Transform Location {type=Vector}
 R2_Previous Velocity Transform Velocity {type=Vector}
 Right_leg_belt {type=number}
 Right_leg_belt_crouch {type=number}
 Skel Mesh {type=SkeletalMeshComponent}
 Velocity Interp Speed {type=number, default=1}
 Velocity Interp Speed_0 {type=number, default=1}
 WPO Pho Wind Intensity {type=number, default=1}
]
